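# Sound cue that plays the looping music and foley for the Hydraulics emote in Athena.
# A mixer layers two looping wave players at equal volume: a worldized music loop and a bounce foley loop.
# It uses the first-person emote music sound class and concurrency rule, with a submix send to the music audio analysis submix.
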SoundCue /Game/Athena/Sounds/Emotes/Hydraulics/Emote_Hydraulics_Music_Cue
root=Mixer_0
nodes [
 {id=Mixer_0, children [WavePlayer_1,WavePlayer_2], properties {inputVolume [1,1]}}
 {id=WavePlayer_1, properties {bLooping=true, soundWaveAssetPtr=&/Emote_Hyrdraulics_Worldized_Loop}}
 {id=WavePlayer_2, properties {bLooping=true, soundWaveAssetPtr=&/Emote_Hydraulics_Bounce_Foley}}
]
properties {
 maxDistance=2097152
 soundClassObject=/Game/Sounds/FortSoundClassesAndModes/Fort_Music_Emote_1P_PSM
 volumeMultiplier=1
 concurrencySet [/Game/Sounds/Concurrency_Rules/Athena_EmoteMusic_1P_Conc]
 soundSubmixSends [{minSendLevel=0, soundSubmix=&Fortnite_AudioAnalysis_Music}]
}
